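# Material function in the FNE UI operations set that checks whether two scalar inputs, Int A and Int B, are equal as integers.
# It floors both inputs, so fractional parts are discarded, then compares them with an If node that outputs 1 when equal and 0 otherwise.
# Both inputs default to preview constants, and the function is exposed to the material library.
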MaterialFunction /FNE_UI/Systems/_MaterialFunctions/Operations/MF_FNEUI_CheckIntEquality
graph {
 properties {bExposeToLibrary=true, description="Check if two Integers are equal. Return 1 if true, 0 if false", libraryCategoriesText [""]}
 nodes [
  n1 = FunctionInput(Preview=Constant()) {bUsePreviewValueAsDefault=true, inputName=Int B, inputType=FunctionInput_Scalar, sortPriority=1}
  n2 = Constant()
  n3 = FunctionInput(Preview=Constant()) {bUsePreviewValueAsDefault=true, inputName=Int A, inputType=FunctionInput_Scalar}
  FunctionOutput(None=If(A=Floor(None=@n3), B=Floor(None=@n1), "A > B"=@n2, "A == B"=Constant() {r=1}, "A < B"=@n2)) {description="Return 1 if value A and value B are Equal. Otherwise, it return 0."}
 ]
}
